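; Interactable ToyFox pedestal building actor that, when a long use completes, dies and tells the game state component to begin the ToyFox enter sequence for the player and squadmates.
; The server knocks back actors on spawn, limits interaction to one team, applies effects to interactors and publishes progress.
; Clients tick only during interaction, smoothing progress into material and audio pitch.
; A timer drives a distance-based, limited-lifetime speech bubble, alongside music and one-shot VO.
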(event EventBeginPlay
  (bind _spawngameplaycuetag_gameplaycuetag (Utilities|Struct|BreakGameplayCueTag (Variables|GetSpawnGameplayCueTag_GameplayCueTag)))
  ;[0 0.15 0.01361 0.5] BeginPlay: Play the spawn in niagara system.
  ; TODO: This will run everytime the actor becomes relevant. Verify if that makes sense once the FX is in.
  (|CallParentFunction)
  (Networking|SwitchHasAuthority
    (:Authority
      ;;[0.06685 0.03576 0.15 0.5] Once the building actor has spawned, knockback all characters near it.
      (Class|AsyncOverlapQueue|AsyncSphereOverlapActors (Transformation|GetActorLocation) (ScalableFloat|GetCachedScalableFloatValue (Variables|Hotfix|GetHotfix_SpawnKnockbackActorsDetectRadius)) (Utilities|Array|MakeArray self) (EventDispatchers|CreateEvent :!event HandleAsyncOverlapSpawnKnocbackShouldCancel) (EventDispatchers|CreateEvent :!event HandleAsyncOverlapSpawnKnockbackHitActor) (EventDispatchers|CreateEvent :!event HandleAsyncOverlapSpawnKnockbackComplete) "ECC_GameTraceChannel3")
      (GameplayCue|ExecuteGameplayCueOnActor(Burst) self _spawngameplaycuetag_gameplaycuetag (Ability|GameplayCue|MakeGameplayCueParameters :NormalizedMagnitude 0.0 :RawMagnitude 0.0 :Location (Transformation|GetSocketLocation (Variables|Default|GetStaticMesh_Base) "Center") :Normal "0,0,0" :GameplayEffectLevel 1 :AbilityLevel 1 :bReplicateLocationWhenUsingMinimalRepProxy false)))
    (:Remote
      (CallFunction|PlayIdleMusic (Transformation|GetSocketLocation (Variables|Default|GetStaticMesh_Base) "Center"))
      (SpeechBubble|SetupSpeechBubbleText)
      (CallFunction|SetSpeechBubbleEnable true)
      ;;[0.03644 0.05971 0.15 0.5] If the actor is being interacted with, start ticking for visual updates.
      (Actor|Tick|SetActorTickEnabled (> (Class|FortActorComponentMultiInteract|GetInteractingActorCount (Variables|Default|GetMultiInteractActorComponent) true true) 0)))))

(event Custom|Play%20Idle%20Music
  (params (Location :type Vector))
  (bind _returnvalue (Audio|SpawnSoundatLocation (Variables|FX|GetMusic%20Loop) Location))
  (bind _output_get (Variables|Default|SetIdle%20Music%20Audio _returnvalue))
  (Audio|AddEventbyComponent (GameInstanceSubsystems|GetMusicEventSubsystem) _output_get (Variables|FX|GetIdle%20Music%20Event%20Tag) "(GameplayTags=)" 0 ""))

(event Interaction|EventOnServerNotifyStartLongUse
  (params (InteractingPawn :type Fort%20Player%20Pawn%20Object%20Reference))
  (bind _returnvalue (Ability|GetFortAbilitySystemComponent InteractingPawn))
  ;[0.15 0.02097 0.1381 0.5] Start/End Long Use: Track the players interacting with the BGA.
  ; This value is also available in MultiInteractActorComponent but it is not blueprint visible.
  (|CallParentFunction InteractingPawn)
  (Networking|SwitchHasAuthority
    (:Authority
      (Utilities|IsValid InteractingPawn
        (:Is%20Valid
          (Utilities|Array|Add (Variables|Transient|GetInteractingPlayerPawns) InteractingPawn)
          (Actor|Tick|SetActorTickEnabled true)
          (Utilities|IsValid _returnvalue
            (:Is%20Valid
              (GameplayEffects|ApplyGameplayEffectToSelf :self _returnvalue :GameplayEffectClass (Variables|Settings|GetPedestalInteractingGameplayEffectClass) :Level 0.0)
              (Tags|CancelAbilitiesWithTags _returnvalue (GameplayTags|MakeLiteralGameplayTagContainer "(GameplayTags=((TagName=\"Gameplay.Action.Player.CarryingPlayer\")))")))))))))

(event Interaction|EventOnServerNotifyEndLongUse
  (params (InteractingPawn :type Fort%20Player%20Pawn%20Object%20Reference) (bUseCompleted :type Boolean))
  (bind _returnvalue (Ability|GetFortAbilitySystemComponent InteractingPawn))
  (bind _interactingplayerpawns (Variables|Transient|GetInteractingPlayerPawns))
  ;[0.15 0.02097 0.1381 0.5] Start/End Long Use: Track the players interacting with the BGA.
  ; This value is also available in MultiInteractActorComponent but it is not blueprint visible.
  (|CallParentFunction InteractingPawn bUseCompleted)
  (Networking|SwitchHasAuthority
    (:Authority
      (Utilities|Array|RemoveItem _interactingplayerpawns InteractingPawn)
      (if (Utilities|Array|IsEmpty _interactingplayerpawns)
        (Actor|Tick|SetActorTickEnabled false)
        (GameState|UpdateGameStateComponentInteractionInfo 0 (struct :InteractingTeam_3_796B5D3849EFF5868A55DDAF7E396640 -1 :InteractionProgress_6_75DA264A4A699250D89AC295A8E123DF -1.0)) ; Reset
        (else
          _))
      (Utilities|IsValid _returnvalue
        (:Is%20Valid
          (GameplayEffects|RemoveActiveGameplayEffectbySourceEffect _returnvalue (Variables|Settings|GetPedestalInteractingGameplayEffectClass) 0 -1))))))

(event EventOnDeathPlayEffects
  (params (Damage :type Float%20%28single-precision%29) (DamageTags :type Gameplay%20Tag%20Container%20Structure%20%28by%20ref%29) (Momentum :type Vector :default "0, 0, 0") (HitInfo :type Hit%20Result%20Structure%20%28by%20ref%29) (InstigatedBy :type Fort%20Pawn%20Object%20Reference) (DamageCauser :type Actor%20Object%20Reference) (EffectContext :type Gameplay%20Effect%20Context%20Handle%20Structure))
  (bind _despawngameplaycuetag_gameplaycuetag (Utilities|Struct|BreakGameplayCueTag (Variables|GetDespawnGameplayCueTag_GameplayCueTag)))
  ;[0.008162 0.1048 0.15 0.5] OnDeathPlayEffects: Play effects on death.
  (|CallParentFunction Damage DamageTags Momentum HitInfo InstigatedBy DamageCauser EffectContext)
  (GameplayCue|ExecuteGameplayCue%28Burst%29Local self _despawngameplaycuetag_gameplaycuetag (Ability|GameplayCue|MakeGameplayCueParameters :NormalizedMagnitude 0.0 :RawMagnitude 0.0 :Location (Transformation|GetSocketLocation (Variables|Default|GetStaticMesh_Base) "Center") :Normal "0,0,0" :GameplayEffectLevel 1 :AbilityLevel 1 :bReplicateLocationWhenUsingMinimalRepProxy false))
  (bind _idle_music_audio (Variables|Default|GetIdle%20Music%20Audio)
    (:then
      (Audio|Components|Audio|FadeOut _idle_music_audio 0.7 0.0 "Linear" :!class /Script/Engine.AudioComponent))
    (:else)))

(event EventOnDeathServer
  (params (Damage :type Float%20%28single-precision%29) (DamageTags :type Gameplay%20Tag%20Container%20Structure%20%28by%20ref%29) (Momentum :type Vector :default "0, 0, 0") (HitInfo :type Hit%20Result%20Structure%20%28by%20ref%29) (InstigatedBy :type Controller%20Object%20Reference) (DamageCauser :type Actor%20Object%20Reference) (EffectContext :type Gameplay%20Effect%20Context%20Handle%20Structure))
  ;[0.09896 0.0487 0 0.5] OnDeathServer: Disable collision on death.
  (|CallParentFunction Damage DamageTags Momentum HitInfo InstigatedBy DamageCauser EffectContext)
  (Variables|Transient|SetbDespawned true))

(event EventTick
  (params (DeltaSeconds :type Float%20%28single-precision%29))
  (bind _interactingplayerpawns (Variables|Transient|GetInteractingPlayerPawns))
  (bind _output (Utilities|Array|Get(acopy) _interactingplayerpawns))
  (bind _multiinteractactorcomponent (Variables|Default|GetMultiInteractActorComponent))
  ;[0.006336 0.05341 0.15 0.5] Tick: On the client, cosmetic update on the actor to show visual interaction progress.
  ; On the server, update the game state with the interaction info.
  (|CallParentFunction DeltaSeconds)
  (Networking|SwitchHasAuthority
    (:Authority
      ;;[0.1356 0.033 0.15 0.5] Update the game state component with the interaction info.
      (if (Utilities|Array|IsValidIndex _interactingplayerpawns)
        (Utilities|IsValid _output
          (:Is%20Valid
            (GameState|UpdateGameStateComponentInteractionInfo 0 (Utilities|Struct|MakeToyFoxPedestalInteractionInfoStruct (Math|Conversions|ToInteger(Byte) (Team|GetActorTeam _output)) (Math|Float|SafeDivide (Class|FortActorComponentMultiInteract|GetServerInteractionDuration (Variables|Default|GetMultiInteractActorComponent)) (ScalableFloat|GetCachedScalableFloatValue (Variables|Hotfix|GetHotfix_InteractionTime))))))
          (:Is%20Not%20Valid
            _))
        (else
          _))
      (GameState|UpdateGameStateComponentInteractionInfo 0 (struct :InteractingTeam_3_796B5D3849EFF5868A55DDAF7E396640 -1 :InteractionProgress_6_75DA264A4A699250D89AC295A8E123DF -1.0)))
    (:Remote
      (bind _output_get (Variables|Transient|SetInteractionProgress (Math|Interpolation|FInterptoConstant (Variables|Transient|GetInteractionProgress) (Math|Float|SafeDivide (Class|FortActorComponentMultiInteract|GetServerInteractionDuration _multiinteractactorcomponent) (ScalableFloat|GetCachedScalableFloatValue (Variables|Hotfix|GetHotfix_InteractionTime))) DeltaSeconds 3.5)))
      (Development|PrintString (Utilities|String|Append "ToyFoxPedestalInteractionCompletion: " (Utilities|String|ToString(Float) _output_get)) true false "(R=0.000000,G=0.660000,B=1.000000,A=1.000000)" 2.0 "ToyFoxPedestalInteractionDuration")
      (Rendering|Material|SetScalarParameterValueonMaterials (Variables|Default|GetStaticMesh_Base) "Interaction_Progession" _output_get)
      (Audio|Components|Audio|SetPitchMultiplier (Variables|Default|GetAudio) (+ 1.0 (* _output_get 0.5))) ; TODO: Better implementaiton of interaction audio.
      ;; Disable ticking once the value reaches 0 and there's no player interacting with the pedestal.
      (if (and (<= _output_get 0) (<= (Class|FortActorComponentMultiInteract|GetInteractingActorCount _multiinteractactorcomponent true true) 0))
        (Actor|Tick|SetActorTickEnabled false)))))

(event ActorCountChangedDelegate(MultiInteractActorComponent)
  (params (InteractingActorCount :type Integer) (MultiInteractMultiplier :type Float%20%28single-precision%29) (InteractingDuration :type Float%20%28single-precision%29))
  ;[0.1379 0.03133 0.15 0.5] MultiInteractActorComponent: Number of interacting actor changed. Start ticking if the number of interacting actor is greater than 0.
  (Networking|CanExecuteCosmeticEvents
    (:True
      (if (> (Class|FortActorComponentMultiInteract|GetInteractingActorCount (Variables|Default|GetMultiInteractActorComponent) true true) 0)
        (Actor|Tick|SetActorTickEnabled true)))))

(event EventEndPlay
  (params (EndPlayReason :type EEndPlayReason%20Enum :default "Destroyed"))
  ;[0.04688 0.02789 0 0.5] EndPlay: Clean up.
  (Utilities|FlowControl|Sequence
    (:then_0
      (Networking|SwitchHasAuthority
        (:Authority
          (for _array_element (Variables|Transient|GetInteractingPlayerPawns)
            (Utilities|IsValid (Ability|GetFortAbilitySystemComponent _array_element)
              (:Is%20Valid
                (GameplayEffects|RemoveActiveGameplayEffectbySourceEffect (Ability|GetFortAbilitySystemComponent _array_element) (Variables|Settings|GetPedestalInteractingGameplayEffectClass) 0 -1)))))))
    (:then_1
      (|CallParentFunction EndPlayReason))))

(event Custom|SetSpeechBubbleEnable
  (params (bEnable :type Boolean))
  ;[0.08108 0.1551 0.276 0.5] SetSpeechBubbleEnable: Sets or disables a timer that updates the speech bubble widget properties for the local player.
  (Networking|CanExecuteCosmeticEvents
    (:True
      (if bEnable
        (if (and (ScalableFloat|GetValueasBool (Variables|Hotfix|GetHotfix_ShowSpeechBubble)) (not (Variables|Transient|GetbSpeechBubbleLifetimeComplete)))
          (bind _returnvalue (Utilities|Time|SetTimerbyEvent (EventDispatchers|CreateEvent :!event UpdateSpeechBubble) (ScalableFloat|GetValueatLevel (Variables|Hotfix|GetHotfix_ShowSpeechBubbleWidgetUpdateFreq)) true true))
          (Variables|Transient|SetUpdateSpeechBubbleWidgetTimerHandle _returnvalue)
          (SpeechBubble|UpdateSpeechBubble))
        (else
          (Utilities|Time|ClearandInvalidateTimerbyHandle (Variables|Transient|GetUpdateSpeechBubbleWidgetTimerHandle))
          (Development|SetHiddeninGame (Variables|Default|GetSpeechBubbleWidget) true false))))))

(event Custom|HandleSpeechBubbleLifetimeEnded
  ;[0 0.1482 0.15 0.5] HandleSpeechBubbleLifetimeEnded:
  (Networking|CanExecuteCosmeticEvents
    (:True
      (CallFunction|SetSpeechBubbleEnable false)
      (Variables|Transient|SetbSpeechBubbleLifetimeComplete true))))

(event Interaction|EventOnLocalInteractStart
  (params (FortPC :type Fort%20Player%20Controller%20Object%20Reference))
  ;[0 0.06136 0.15 0.5] OnLocalInteractStart: Once you get close to the pedestal and highlight it, this VO plays once.
  (|CallParentFunction FortPC)
  (Networking|CanExecuteCosmeticEvents
    (:True
      (Utilities|FlowControl|DoOnce false
        (:Completed
          (Audio|PlaySoundatLocation (Variables|VO|GetZordonPedestalVO) (Transformation|GetActorLocation)))))))

(fn BlueprintCanInteract
  (params (InteractingPawn :type Fort%20Pawn%20Object%20Reference) (InteractionBeingAttempted :type EInteractionBeingAttempted%20Enum :default "FirstInteraction") (InteractionType :type TInteractionType%20Enum :default "IT_NoInteraction"))
  (bind _interactingplayerpawns (Variables|Transient|GetInteractingPlayerPawns))
  (bind _output (Utilities|Array|Get(acopy) _interactingplayerpawns))
  ;[0.03086 0.1284 0.15 0.5] CanInteract: Return if the player can interact based on their state and team.
  (bind _asfort_player_pawn_athena (Utilities|Casting|CastToFortPlayerPawnAthena InteractingPawn)
    (:then
      (Utilities|FlowControl|Sequence
        (:then_0
          ;;[0.07272 0.02373 0.15 0.5] Cannot interaction if the actor is already being interacted with by another team.
          (if (Utilities|Array|IsValidIndex _interactingplayerpawns)
            (Utilities|IsValid _output
              (:Is%20Valid
                (if (!= (Team|GetActorTeam _output) (Team|GetActorTeam InteractingPawn))
                  (return false))))))
        (:then_1
          ;;[0.15 0.07114 0.03616 0.5] Can interact if the player is not DBNO or in a vehicle.
          (return (Math|Boolean|ANDBoolean (Variables|Interaction|GetbAllowInteract) (Variables|Transient|GetbEnableInteract) (not (Vehicle|IsinVehicle _asfort_player_pawn_athena)) (not (Health|IsDBNO _asfort_player_pawn_athena)))))))
    (:CastFailed
      (return false))))

(fn BlueprintGetFailedInteractionString
  (params (InteractingPawn :type Fort%20Pawn%20Object%20Reference) (InteractionBeingAttempted :type EInteractionBeingAttempted%20Enum :default "FirstInteraction"))
  (return "NSLOCTEXT(\"[AC15B824F2F9671D8AA196543A7B62FC]\", \"5D10E57E4AC63D928E313B92EAEE15B1\", \"Already In Use\")"))

(fn BlueprintGetInteractionString
  (params (InteractingPawn :type Fort%20Pawn%20Object%20Reference) (InteractionBeingAttempted :type EInteractionBeingAttempted%20Enum :default "FirstInteraction"))
  ;[0.05539 0.09564 0.15 0.5] Returns interaction string based on the playlist type.
  (return (Variables|Interaction|GetInteractionText_Default)))

(fn BlueprintGetInteractionTime
  (params (InteractingPawn :type Fort%20Pawn%20Object%20Reference) (InteractionBeingAttempted :type EInteractionBeingAttempted%20Enum :default "FirstInteraction"))
  (return (ScalableFloat|GetCachedScalableFloatValue (Variables|Hotfix|GetHotfix_InteractionTime)) true))

(fn BlueprintOnInteract
  (params (InteractingPawn :type Fort%20Pawn%20Object%20Reference) (InteractionBeingAttempted :type EInteractionBeingAttempted%20Enum :default "FirstInteraction"))
  (bind _interactingplayerpawns (Variables|Transient|GetInteractingPlayerPawns))
  (bind _output (Utilities|Array|Get(acopy) _interactingplayerpawns))
  ;[0 0.15 0.003426 0.5] BlueprintOnInteract: When interaction is successul, let the game state component know so it can begin the toyfox event sequence.
  (|CallParentFunction InteractingPawn InteractionBeingAttempted)
  (Networking|SwitchHasAuthority
    (:Authority
      (LocalVariables|SetPrimaryInteractingPlayer (Utilities|Casting|CastToFortPlayerPawn InteractingPawn))
      (if (Utilities|Array|IsValidIndex _interactingplayerpawns)
        (Utilities|IsValid _output
          (:Is%20Valid
            (LocalVariables|SetPrimaryInteractingPlayer _output))
          (:Is%20Not%20Valid
            _))
        (bind _primaryinteractingplayer (LocalVariables|GetPrimaryInteractingPlayer)
          (:then
            (Variables|Transient|SetbEnableInteract false)
            (Gameplay|Die)
            (bind _component (GameState|GetGameStateComponent))
            (Utilities|IsValid _component
              (:Is%20Valid
                (bind _squadplayerpawns (CallFunction|GetSquadmates _primaryinteractingplayer))
                (Class|BToyFoxGameStateComponent|BeginToyFoxEnterSequence _component _primaryinteractingplayer _squadplayerpawns false))))
          (:else))
        (else
          _)))))

(fn GetFocusedSocketLocation
  (return (+ (Transformation|GetActorLocation) "0.000000,0.000000,100.000000")))

(fn GetGameStateComponent
  ;[0 0.15 0.1226 0.5] Return the game state component. Cache it the first time it is found.
  (Variables|Transient|GetGameStateComponent
    (:then
      _)
    (:else
      (bind _returnvalue (Fortnite|Athena|GameState|GetGameStateAthena))
      (Utilities|IsValid _returnvalue
        (:Is%20Valid
          (bind _returnvalue_2 (Actor|GetComponentbyClass _returnvalue "/ToyFoxGameplay/Gameplay/B_ToyFoxGameStateComponent.B_ToyFoxGameStateComponent_C"))
          (Utilities|IsValid _returnvalue_2
            (:Is%20Valid
              (bind _output_get (Variables|Transient|SetGameStateComponent _returnvalue_2))))))))
  (return _output_get))

(fn GetMarkerDetailsId
  (return "None"))

(fn GetSquadmates
  (params (PrimaryPlayer :type Fort%20Player%20Pawn%20Object%20Reference))
  (Utilities|FlowControl|Sequence
    (:then_0
      ;[0.01484 0.15 0.1485 0.5] Returns the squadmates of the primary player.
      (Networking|SwitchHasAuthority
        (:Authority
          (LocalVariables|SetbForceReviveDBNO (> (ScalableFloat|GetCachedScalableFloatValue (Variables|Hotfix|GetHotfix_ForceReviveDBNO)) 0))
          (Utilities|IsValid PrimaryPlayer
            (:Is%20Valid
              (for _array_element (Fortnite|GetAllFortPlayerControllers true)
                (bind _returnvalue (Game|Player|GetPlayerPawn _array_element))
                (Utilities|IsValid _returnvalue
                  (:Is%20Valid
                    (bind _returnvalue_2 (Fortnite|Utilities|AttemptGetPlayerStateforPawn _returnvalue))
                    (bind _asfort_player_state_athena (Utilities|Casting|CastToFortPlayerStateAthena _returnvalue_2))
                    (if (Math|Boolean|ANDBoolean (Class|FortPlayerStateAthena|IsAnAthenaGameParticipant _asfort_player_state_athena) (== (Class|FortPawn|GetTeam _returnvalue) (Class|FortPawn|GetTeam PrimaryPlayer)) (or (not (Health|IsDBNO _returnvalue)) (LocalVariables|GetbForceReviveDBNO)) (!= _returnvalue PrimaryPlayer))
                      (Utilities|Array|Add (LocalVariables|GetRetArray) _returnvalue))))))))))
    (:then_1
      (return (LocalVariables|GetRetArray)))))

(fn HandleAsyncOverlapSpawnKnocbackShouldCancel
  (return (Game|IsActorBeingDestroyed)))

(fn HandleAsyncOverlapSpawnKnockbackComplete
  (params (WasCancelled :type Boolean)))

(fn HandleAsyncOverlapSpawnKnockbackHitActor
  (params (HitActor :type Actor%20Object%20Reference))
  ;[0.1283 0.04173 0.15 0.5] Knockback any overlapping actors on spawn.
  (Networking|SwitchHasAuthority
    (:Authority
      (Utilities|IsValid HitActor
        (:Is%20Valid
          (Fortnite|ApplyGameplayEffectsToActor HitActor (Utilities|Array|MakeArray (Variables|Knockback|GetSpawnKnockbackGameplayEffect)) (Transformation|GetActorLocation) false 0 (Ability|GetFortAbilitySystemComponent self) (GameplayEffects|MakeFortEffectContext (Ability|GetFortAbilitySystemComponent self) self self 0 0))
          (bind _asfort_player_pawn (Utilities|Casting|CastToFortPlayerPawn HitActor))
          (Collision|AddIgnoredActor _asfort_player_pawn self 0.33))))))

(fn OnRep_bDespawned
  ;[0 0.1442 0.15 0.5] OnRep_bDespawned: Set hidden and remove collision.
  ; Called on server and client.
  (if (Variables|Transient|GetbDespawned)
    (Collision|SetActorEnableCollision false)
    (Rendering|SetActorHiddenInGame true)
    (Audio|Components|Audio|FadeOut (Variables|Default|GetAudio) 0.2 0.0 "Linear" :!class /Script/Engine.AudioComponent)))

(fn SetupSpeechBubbleText
  ;[0.009664 0.02195 0.15 0.5] Set up the speech buble with the desired text.
  (Networking|CanExecuteCosmeticEvents
    (:True
      (bind _asui_npccore_speech_bubble (Utilities|Casting|CastToUI_NPCCore_SpeechBubble (UserInterface|GetUserWidgetObject (Variables|Default|GetSpeechBubbleWidget))))
      (bind _output_get (Variables|Transient|SetSpeechBubbleWidgetObject _asui_npccore_speech_bubble))
      (Conversation|SetPreviewMessage _output_get (Utilities|Text|FormatText "NSLOCTEXT(\"[AC15B824F2F9671D8AA196543A7B62FC]\", \"EEC9430A45447C98A3E7A4B2742FF77D\", \"<HighlightB0>{Message}</>\")" (Variables|SpeechBubble|GetSpeechBubbleText))))))

(fn UpdateGameStateComponentInteractionInfo
  (params (Info :type Toy%20Fox%20Pedestal%20Interaction%20Info%20Struct%20Structure))
  ;[0.002725 0.0681 0.15 0.5] Update the game state component with information regarding the pedestal interaction.
  (Networking|SwitchHasAuthority
    (:Authority
      (bind _component (GameState|GetGameStateComponent))
      (Utilities|IsValid _component
        (:Is%20Valid
          (Pedestal|SetPedestalInteractionInfo _component Info))))))

(fn UpdateSpeechBubble
  (bind _speechbubblewidget (Variables|Default|GetSpeechBubbleWidget))
  (bind _speechbubblewidget_2 (Variables|Default|GetSpeechBubbleWidget))
  ;[0.1484 0.04088 0.15 0.5] UpdateSpeechBubble: Update the visibility of the speech bubble and its transform settings based on the distance of the closest local player pawn.
  (Networking|CanExecuteCosmeticEvents
    (:True
      (Variables|Default|GetSpeechBubbleWidget
        (:then
          (Utilities|FlowControl|Sequence
            (:then_0
              ;; Find the closest local player pawn.
              (for _array_element (Fortnite|GetLocalFortPlayerControllers true)
                (Utilities|IsValid _array_element
                  (:Is%20Valid
                    (Utilities|Array|Add (LocalVariables|GetLocalPlayerPawns) (Game|Player|GetPlayerPawn _array_element)))))
              (LocalVariables|SetClosestLocalPlayerPawn (Fortnite|GetClosestActorfromArray self (LocalVariables|GetLocalPlayerPawns))))
            (:then_1
              (bind _closestlocalplayerpawn (LocalVariables|GetClosestLocalPlayerPawn)
                (:then
                  (bind _output_get (LocalVariables|SetbShowSpeechBubbleWidget (< (Math|Vector|DistanceSquared(Vector) (Transformation|GetActorLocation) (Transformation|GetActorLocation _closestlocalplayerpawn)) (Math|Float|Square (ScalableFloat|GetCachedScalableFloatValue (Variables|Hotfix|GetHotfix_ShowSpeechBubbleMaxDistance))))))
                  (if _output_get
                    ;; Show the speech bubble and update its transform if there's a local player pawn within range.
                    (if (Variables|/Script/Engine.SceneComponent|GetbHiddenInGame _speechbubblewidget_2)
                      (Development|SetHiddeninGame _speechbubblewidget_2 false false)
                      (SpeechBubble|SetupSpeechBubbleText) ; This retriggers the intro animation on the bubble
                      (else
                        _))
                    (bind _speechbubblewidgetobject (Variables|Transient|GetSpeechBubbleWidgetObject)
                      (:then
                        (Variables|/Script/UMG.Widget|SetRenderTransform_Translation_X (Utilities|Struct|MakeWidgetTransform (Math|Vector2D|MakeVector2D 0.0 (Math|Float|Lerp 0.0 -1e+02 (Math|Float|NormalizetoRange (Math|Vector|Distance(Vector) (Transformation|GetActorLocation (LocalVariables|GetClosestLocalPlayerPawn)) (Transformation|GetActorLocation)) 0.0 5e+03)))) _speechbubblewidgetobject)
                        ;;[0.03542 0.1161 0.15 0.5] Start lifetime timer handle as soon as it is shown once.
                        (if (not (Utilities|Time|IsTimerActivebyHandle (Variables|Transient|GetSpeechBubbleLifetimeTimerHandle)))
                          (bind _returnvalue (Utilities|Time|SetTimerbyEvent (EventDispatchers|CreateEvent :!event HandleSpeechBubbleLifetimeEnded) (ScalableFloat|GetValueatLevel (Variables|Hotfix|GetHotfix_ShowSpeechBubbleLifetime)) false true))
                          (Variables|Transient|SetSpeechBubbleLifetimeTimerHandle _returnvalue)))
                      (:else))
                    (elif (not (Variables|/Script/Engine.SceneComponent|GetbHiddenInGame _speechbubblewidget))
                      ;; Hide the speech bubble if there's no valid local player OR all the local players are out of range.
                      (Development|SetHiddeninGame _speechbubblewidget true false))))
                (:else
                  _)))))
        (:else)))))

(fn ConstructionScript)
