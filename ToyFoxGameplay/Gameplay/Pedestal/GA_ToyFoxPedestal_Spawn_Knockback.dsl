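; Knockback gameplay ability for the ToyFox pedestal spawn, by its name a child of a launch-style base ability.
; It overrides FindLauncher to return the effect causer from the context of the effect that granted the ability, ignoring the DefaultLauncher argument.
; The activate and end events hold no active logic, so any knockback behavior comes from outside this graph.
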(event Ability|EventActivateAbility
  (|CallParentFunction) ; This node is disabled and will not be called.\nDrag off pins to build functionality.
)

(event Ability|EventOnEndAbility
  (params (bWasCancelled :type Boolean)))

(fn FindLauncher
  (params (DefaultLauncher :type Actor%20Object%20Reference))
  (return (Ability|EffectContext|GetEffectCauser (Ability|GetGrantedbyEffectContext))))
